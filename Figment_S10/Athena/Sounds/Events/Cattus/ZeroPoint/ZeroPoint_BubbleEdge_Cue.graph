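# ZeroPoint_BubbleEdge_Cue is a looping ambient SoundCue, named for the edge of the Zero Point bubble in the Cattus event.
# A Mixer layers two looping waves, a sphere-edge loop and a repurposed hoverboard movement loop, at equal volume.
# The boosted cue uses a high-priority events sound class.
# Its attenuation override adds logarithmic falloff, distance low-pass filtering, wide stereo spread and interpolated listener-focus settings that favour sounds being faced.
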SoundCue /Figment_S10/Athena/Sounds/Events/Cattus/ZeroPoint/ZeroPoint_BubbleEdge_Cue
root=Mixer_1
nodes [
 {id=Mixer_1, children [WavePlayer_0,WavePlayer_3], properties {inputVolume [1,1]}}
 {id=WavePlayer_0, properties {bLooping=true, soundWaveAssetPtr=/Game/Athena/Sounds/Events/Cattus/ZeroPoint/Snow_Expl_SphereEdge_Loop_01}}
 {id=WavePlayer_3, properties {bLooping=true, soundWaveAssetPtr=/Game/Athena/Sounds/Vehicles/Hoverboard/Hoverboard_MoveLoop_06}}
]
properties {
 bOverrideAttenuation=true
 maxDistance=1000
 soundClassObject=/Figment_S10/Athena/Sounds/Events/SoundClasses/Fort_Events_Priority_ExtHigh
 volumeMultiplier=2
 attenuationOverrides {
  bAttenuateWithLPF=true
  bEnableFocusInterpolation=true
  distanceAlgorithm=Logarithmic
  falloffDistance=1000
  focusAttackInterpSpeed=0.5
  focusAzimuth=15
  focusDistanceScale=0.5
  focusReleaseInterpSpeed=0.25
  lPFFrequencyAtMax=500
  lPFRadiusMax=2000
  lPFRadiusMin=1000
  nonFocusAzimuth=90
  nonFocusVolumeAttenuation=0.1
  stereoSpread=512
  attenuationShapeExtents {x=0}
  customAttenuationCurve {editorCurveData {keys [<at 0 1>,<at 0.3 1>,<at 0.4 0.5>,<at 1 0>]}}
 }
}
