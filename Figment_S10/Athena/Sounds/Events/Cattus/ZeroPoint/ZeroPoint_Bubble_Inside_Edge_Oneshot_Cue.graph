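# Sound cue for a one-shot at the inside edge of the Cattus Zero Point bubble, judging by its name.
# An equally weighted Random node picks one of four distant rift energy waves borrowed from RiftBeacon assets.
# Attenuation overrides supersede the assigned RiftBeacon outside-ambience asset, turning off volume falloff while enabling distance low-pass filtering and a wide stereo spread.
# It plays in the extra-high-priority events sound class.
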SoundCue /Figment_S10/Athena/Sounds/Events/Cattus/ZeroPoint/ZeroPoint_Bubble_Inside_Edge_Oneshot_Cue
root=Random_1
nodes [
 {id=Random_1, children [WavePlayer_1,WavePlayer_4,WavePlayer_5,WavePlayer_6], properties {weights [1,1,1,1]}}
 {id=WavePlayer_1, properties {soundWaveAssetPtr=/Game/Athena/Sounds/Events/RiftBeacon/AssetDrop_01/Rift_Energy_Dist_04}}
 {id=WavePlayer_4, properties {soundWaveAssetPtr=/Game/Athena/Sounds/Events/RiftBeacon/AssetDrop_01/Rift_Energy_Dist_01}}
 {id=WavePlayer_5, properties {soundWaveAssetPtr=/Game/Athena/Sounds/Events/RiftBeacon/AssetDrop_01/Rift_Energy_Dist_02}}
 {id=WavePlayer_6, properties {soundWaveAssetPtr=/Game/Athena/Sounds/Events/RiftBeacon/AssetDrop_01/Rift_Energy_Dist_03}}
]
properties {
 attenuationSettings=/Game/Athena/Sounds/Events/RiftBeacon/RiftBeacon_Bubble_Amb_Outside_Att
 bOverrideAttenuation=true
 maxDistance=2097152
 soundClassObject=/Figment_S10/Athena/Sounds/Events/SoundClasses/Fort_Events_Priority_ExtHigh
 volumeMultiplier=1
 attenuationOverrides {bAttenuate=false, bAttenuateWithLPF=true, lPFFrequencyAtMax=200, lPFRadiusMax=4000, lPFRadiusMin=2000, stereoSpread=4096}
}
