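# Looping Sound Cue for the inside of the Zero Point in the Cattus event.
# A single looping Wave Player plays the Snow_Expl_2D_Loop_01 wave at boosted volume in an extra-high priority events sound class.
# Attenuation overrides disable distance falloff and spatialization, so it plays as 2D audio despite the assigned attenuation asset.
# Concurrency overrides allow one instance at a time and prevent new ones from starting.
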SoundCue /Figment_S10/Athena/Sounds/Events/Cattus/ZeroPoint/ZeroPoint_Inside_Loop_Cue
root=WavePlayer_0
nodes [{id=WavePlayer_0, properties {bLooping=true, soundWaveAssetPtr=&/Snow_Expl_2D_Loop_01}}]
properties {
 attenuationSettings=/Figment_S10/Sounds/Attenuation/Fort_LowGravityStones_Att
 bOverrideAttenuation=true
 bOverrideConcurrency=true
 maxDistance=2097152
 soundClassObject=/Figment_S10/Athena/Sounds/Events/SoundClasses/Fort_Events_Priority_ExtHigh
 volumeMultiplier=1.5
 attenuationOverrides {bAttenuate=false, bSpatialize=false, falloffDistance=10000}
 concurrencyOverrides {maxCount=1, resolutionRule=PreventNew}
}
